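# Physics asset for the skeletal mesh of the Sharp Dresser pickaxe, a two-handed melee weapon cosmetic.
# The handle bone carries a long capsule body and the dyn_loop bone carries a box body.
# A single constraint joins the loop to the handle with equal limits on both swings and twist, bounding how far the loop can rotate relative to the handle.
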PhysicsAsset /BRCosmetics/Weapons/FORT_Melee/Pickaxe_SharpDresser/Meshes/SK_Pickaxe_SharpDresser_PhysicsAsset
bodies {
 Melee_TwoHanded_Handle [{shapeType=Capsule, length=96.1552, radius=4.6207, center=<0.1245 -1.4568 27.0604>, rotation=<rot 0.5051 -90 -0.1267>}]
 dyn_loop [{shapeType=Box, extentX=2.9232, extentY=10.5873, extentZ=14.9874, center=<0.011 0.0787 -11.7468>, rotation=<rot -0.01 -90 0.1218>}]
}
constraints [{bone1Name=dyn_loop, bone2Name=Melee_TwoHanded_Handle, swing1LimitDegrees=60, swing2LimitDegrees=60, twistLimitDegrees=60}]
